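# Niagara mesh particle material for debris effects of the Emperor Male pickaxe, using Default Lit shading.
# It samples the pickaxe diffuse texture with uniform UV scale and offset from TextureUV's R and B channels, then applies power, desaturation and color tint.
# Particle color multiplies the result, with an EmissiveScale branch routed through MF_FX_EmissiveAdjustments.
# An MR vector parameter is also defined.
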Material /BRCosmetics/Weapons/FORT_Melee/Pickaxe_Emperor_Male/FX/M_Pickaxe_Emperor_Male_Debris
graph {
 properties {bUsedWithNiagaraMeshParticles=true, pixelDepthOffsetMode=PDOM_Legacy, usedShadingModels=Default Lit}
 nodes [
  n1 = ParticleColor()
  n2 = ScalarParameter() {parameterName=EmissiveScale}
  n3 = TextureCoordinate()
  n4 = VectorParameter() {parameterName=TextureUV, defaultValue=<rgba 0 0 0 1>}
  n5 = AppendVector(A=@n4.R, B=@n4.R)
  n6 = TextureSampleParameter2D(UVs=Add(A=Multiply(A=@n3, B=@n5), B=AppendVector(A=@n4.B, B=@n4.B))) {parameterName=Tex, texture=/BRCosmetics/Weapons/FORT_Melee/Pickaxe_Emperor_Male/Textures/T_Pickaxe_Emperor_Male_D}
  n7 = ScalarParameter() {parameterName=Power}
  n8 = Multiply(
   A=Desaturation(None=Power(Base=@n6.RGB, Exp=@n7), Fraction=ScalarParameter() {parameterName=Desaturation})
   B=VectorParameter() {parameterName=ScaleColor, defaultValue=<rgba 0 0 0 1>}.RGB
  )
  MaterialFunctionCall(Emissive=Multiply(A=@n1.RGB, B=Multiply(A=@n2, B=@n8))) {materialFunction=&MF_FX_EmissiveAdjustments}
  Multiply(A=@n1.RGB, B=@n8)
  VectorParameter() {parameterName=MR}
 ]
}
