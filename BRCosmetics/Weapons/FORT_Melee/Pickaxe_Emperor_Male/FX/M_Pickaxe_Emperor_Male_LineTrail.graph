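# Unlit additive material for Niagara ribbons, named as the line trail effect of the Emperor Male pickaxe.
# An electric noise texture, with parameter-driven UV offset and scale, is blended between power-shaped and linear versions, tinted by particle color and passed through an emissive adjustment function.
# A procedural UV mask peaks along the ribbon's center line, fades in sharply and tapers along its length, scaled by particle alpha.
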Material /BRCosmetics/Weapons/FORT_Melee/Pickaxe_Emperor_Male/FX/M_Pickaxe_Emperor_Male_LineTrail
graph {
 properties {bUsedWithNiagaraRibbons=true, blendMode=BLEND_Additive, pixelDepthOffsetMode=PDOM_Legacy, shadingModel=MSM_Unlit, usedShadingModels=Unlit}
 nodes [
  n1 = TextureCoordinate()
  n2 = VectorParameter()
  n3 = AppendVector(A=@n2.R, B=@n2.R)
  n4 = TextureSampleParameter2D(UVs=Multiply(A=Add(A=@n1, B=@n3), B=AppendVector(A=@n2.B, B=@n2.B))) {
   parameterName=Param
   samplerType=SAMPLERTYPE_LinearColor
   texture=/Game/Characters/Player/Male/Medium/Bodies/M_MED_SRDrift/Textures/T_ElectricNoise
  }
  n5 = VectorParameter() {parameterName=REF}
  n6 = Multiply(A=Power(Base=@n4.RGB, Exp=@n5.R), B=@n5.G)
  n7 = Multiply(A=@n4.RGB, B=@n5.B)
  n8 = ParticleColor()
  MaterialFunctionCall(Emissive=Multiply(A=LinearInterpolate(A=@n6, B=@n7, Alpha=@n5.A), B=@n8.RGB)) {materialFunction=&MF_FX_EmissiveAdjustments}
  n9 = MaterialFunctionCall(Float2=TextureCoordinate()) {materialFunction=&BreakOutFloat2Components}
  n10 = Multiply(A=Multiply(A=@n9.G) {constB=2}, B=Multiply(A=OneMinus(None=@n9.G)) {constB=2})
  n11 = Power(Base=OneMinus(None=@n10)) {constExponent=0.4}
  n12 = Saturate(None=Multiply(A=@n9.R) {constB=40})
  n13 = OneMinus(None=Power(Base=Multiply(A=@n9.R)) {constExponent=0.25})
  Multiply(A=Multiply(A=Multiply(A=@n10) {constB=2}, B=OneMinus(None=@n11)), B=Multiply(A=Multiply(A=@n12, B=@n13), B=@n8.A))
 ]
}
