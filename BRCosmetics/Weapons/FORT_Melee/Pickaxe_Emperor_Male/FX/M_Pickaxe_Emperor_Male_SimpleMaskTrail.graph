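# Unlit, additive, two-sided trail material for the male Emperor pickaxe, set up for Niagara ribbons.
# A curve atlas row sampled along UV V supplies color and alpha, which are multiplied by particle color.
# The color product passes through an emissive-adjustment function.
# The alpha product is masked by an inverted power curve of UV U, so it fades as U increases.
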Material /BRCosmetics/Weapons/FORT_Melee/Pickaxe_Emperor_Male/FX/M_Pickaxe_Emperor_Male_SimpleMaskTrail
graph {
 properties {
  bUsedWithNiagaraRibbons=true
  blendMode=BLEND_Additive
  pixelDepthOffsetMode=PDOM_Legacy
  shadingModel=MSM_Unlit
  twoSided=true
  usedShadingModels=Unlit
 }
 nodes [
  n1 = ParticleColor()
  n2 = CurveAtlasRowParameter(CurveTime=MaterialFunctionCall(Float2=TextureCoordinate()) {materialFunction=&BreakOutFloat2Components}.G) {
   atlas=/Game/Weapons/FORT_Melee/Pickaxe_Emperor_Male/FX/Pickaxe_Emperor_Male_SimpleMaskAtlas
   curve=/Game/Weapons/FORT_Melee/Pickaxe_Emperor_Male/FX/Pickaxe_Emperor_Male_SimpleMaskAtlas_Curve
  }
  MaterialFunctionCall(Emissive=Multiply(A=@n1.RGB, B=@n2)) {materialFunction=&MF_FX_EmissiveAdjustments}
  n3 = Multiply(A=MaterialFunctionCall(Float2=TextureCoordinate()) {materialFunction=&BreakOutFloat2Components}.R)
  Multiply(A=Multiply(A=@n1.A, B=@n2.A), B=OneMinus(None=Power(Base=@n3) {constExponent=0.45}))
 ]
}
